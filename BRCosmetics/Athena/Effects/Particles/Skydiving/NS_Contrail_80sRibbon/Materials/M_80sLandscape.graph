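# Translucent unlit material for Niagara mesh particles in the 80s Ribbon skydiving contrail, drawing a scrolling neon grid landscape.
# A panned grid texture plus square edge mask blends base color toward particle color, with a distant tint along U.
# Two panned Perlin noise layers displace vertices along instance Z, flattened down the middle for a valley.
# Opacity follows the grid, fading along U and by camera vertical angle; a dynamic parameter supplies pan time.
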Material /BRCosmetics/Athena/Effects/Particles/Skydiving/NS_Contrail_80sRibbon/Materials/M_80sLandscape
graph {
 properties {
  bUsedWithNiagaraMeshParticles=true
  blendMode=BLEND_Translucent
  pixelDepthOffsetMode=PDOM_Legacy
  shadingModel=MSM_Unlit
  twoSided=true
  usedShadingModels=Unlit
 }
 nodes [
  Constant3Vector() {constant=<rgba 0 0 1 0>}
  n1 = VectorParameter() {parameterName=BaseColor, defaultValue=<rgba 0 0 0.2 0>}
  n2 = ParticleColor()
  n3 = Panner(Time=DynamicParameter() {paramNames [PanTime,Param2,Param3,Param4]}.PanTime) {speedX=-0.32}
  n4 = Saturate(None=Add(
    A=TextureSample(UVs=@n3) {texture=/BRCosmetics/Athena/Effects/Particles/Skydiving/NS_Contrail_80sRibbon/Textures/T_Grid}.G
    B=NamedRerouteUsage()
   ))
  n5 = TextureCoordinate()
  MaterialFunctionCall(Emissive=LinearInterpolate(
    A=VectorParameter() {parameterName=DistantColor, defaultValue=<rgba 0.1964 0.1044 0.5 0>}.RGB
    B=LinearInterpolate(A=@n1.RGB, B=@n2.RGB, Alpha=@n4)
    Alpha=ComponentMask(None=@n5) {r=true}
   )) {materialFunction=&MF_FX_EmissiveAdjustments}
  n6 = ComponentMask(None=TextureCoordinate()) {r=true}
  n7 = Saturate(None=Multiply(A=OneMinus(None=@n6)) {constB=5})
  n8 = ComponentMask(None=TextureCoordinate()) {g=true}
  n9 = OneMinus(None=@n8)
  n10 = Power(Base=Multiply(A=Multiply(A=@n8, B=@n9)) {constB=4}) {constExponent=4}
  n11 = Panner(Coordinate=TextureCoordinate() {uTiling=1.75, vTiling=1.75}, Time=DynamicParameter() {paramNames [PanTime,Param2,Param3,Param4]}.PanTime) {speedX=-0.44, speedY=0.03}
  n12 = Transform(None=Constant3Vector() {constant=<rgba 0 0 1 0>}) {transformSourceType=TRANSFORMSOURCE_Instance}
  n13 = Multiply(
   A=ConstantBiasScale(None=TextureSample(UVs=@n11) {samplerType=SAMPLERTYPE_LinearColor, texture=/Game/Global/Textures/Default/Tileable_Patterns/T_PerlinNoise_Uncompressed}.G) {bias=-0.5, scale=2}
   B=@n12
  )
  n14 = ScalarParameter() {defaultValue=5, parameterName=HighFrequencyAmount}
  n15 = TextureCoordinate() {uTiling=0.65, vTiling=0.65}
  n16 = DynamicParameter() {paramNames [PanTime,Param2,Param3,Param4]}
  n17 = Multiply(
   A=TextureSample(UVs=Panner(Coordinate=@n15, Time=@n16.PanTime) {speedX=-0.44, speedY=0.03}) {samplerType=SAMPLERTYPE_LinearColor, texture=&T_PerlinNoise_Uncompressed}.G
   B=@n12
  )
  n18 = ScalarParameter() {defaultValue=50, parameterName=LowFrequencyAmount}
  Multiply(
   A=Multiply(A=LinearInterpolate(Alpha=@n7) {constA=0.25}, B=OneMinus(None=@n10))
   B=Add(A=Multiply(A=@n13, B=@n14), B=Multiply(A=@n17, B=@n18))
  )
  n19 = ScalarParameter() {defaultValue=0.3, parameterName=BaseOpacity}
  n20 = Multiply(A=ComponentMask(None=TextureCoordinate()) {r=true}) {constB=6}
  Multiply(A=Multiply(A=Add(A=@n4, B=@n19), B=Saturate(None=@n20)), B=NamedRerouteUsage())
  n21 = ComponentMask(None=TextureCoordinate()) {g=true}
  n22 = OneMinus(None=@n21)
  n23 = Constant() {r=1}
  n24 = If(
   A=Multiply(A=Multiply(A=@n21, B=@n22)) {constB=4}
   B=ScalarParameter() {defaultValue=0.02, parameterName=IfClip}
   "A > B"=Constant()
   "A == B"=@n23
   "A < B"=@n23
  )
  n25 = ComponentMask(None=TextureCoordinate()) {r=true}
  n26 = OneMinus(None=@n25)
  n27 = Constant() {r=1}
  n28 = If(
   A=Multiply(A=Multiply(A=@n25, B=@n26)) {constB=4}
   B=ScalarParameter() {defaultValue=0.02, parameterName=IfClip}
   "A > B"=Constant()
   "A == B"=@n27
   "A < B"=@n27
  )
  NamedRerouteDeclaration(None=Saturate(None=Add(A=@n24, B=@n28))) {name=SquareMask, nodeColor=<rgba 0 0.1562 1 1>}
  n29 = Add(A=DotProduct(
    A=MaterialFunctionCall() {materialFunction=&CameraDirectionVector}.Result
    B=Constant3Vector() {constant=<rgba 0 0 -1 0>}
   ))
  NamedRerouteDeclaration(None=Power(Base=Saturate(None=@n29))) {name=UpFade, nodeColor=<rgba 0.6406 1 0 1>}
 ]
}
