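# Translucent unlit Niagara ribbon material for the 80s-style skydiving contrail, showing a glowing grid over a starfield.
# Stars tile and pan in screen space to blend low and high colors, staying independent of ribbon UVs.
# A tinted grid scrolls on UV1 with optional axis swizzle, is added on top, then multiplied by particle color.
# Opacity combines particle alpha, a squared U-edge ramp, camera fade and depth fade, which mobile skips.
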Material /BRCosmetics/Athena/Effects/Particles/Skydiving/NS_Contrail_80sRibbon/Materials/M_80sRibbon
graph {
 properties {bUsedWithNiagaraRibbons=true, blendMode=BLEND_Translucent, pixelDepthOffsetMode=PDOM_Legacy, shadingModel=MSM_Unlit, usedShadingModels=Unlit}
 nodes [
  Constant() {r=1}
  n1 = VectorParameter() {parameterName=Color Low, defaultValue=<rgba 0.0281 0.0058 0.1 0>}
  n2 = VectorParameter() {parameterName=Color High, defaultValue=<rgba 1.6667 3.3333 5 0>}
  n3 = ScreenPosition()
  n4 = MaterialFunctionCall(Float2=Multiply(A=Constant2Vector() {g=0.9, r=1.6}, B=ScalarParameter() {defaultValue=1, parameterName=TileStars}) {constA=1}) {materialFunction=&BreakOutFloat2Components}
  n5 = ScalarParameter() {defaultValue=0.2, parameterName="Pan V (S)"}
  n6 = ChannelMaskParameter(None=TextureSampleParameter2D(UVs=MaterialFunctionCall!(UV=@n3.ViewportUV, Tile U=@n4.R, Tile V=@n4.R, Pan U, Pan V=@n5, Offset U, Offset V, Time) {materialFunction=/Game/Effects/Fort_Effects/Materials/_Functions/MF_PanTileOffset}.UV's) {parameterName=Stars Texture, texture=/BRCosmetics/Athena/Effects/Particles/Skydiving/NS_Contrail_80sRibbon/Textures/T_TilingStars}.RGBA) {maskChannel=Green, parameterName=Stars Texture Channel, defaultValue=<rgba 0 1 0 0>}
  n7 = VectorParameter() {parameterName=GridColor, defaultValue=<rgba 3 0 0.25 0>}
  n8 = ChannelMaskParameter(None=TextureSampleParameter2D(UVs=NamedRerouteUsage()) {parameterName=Grid Texture, texture=/BRCosmetics/Athena/Effects/Particles/Skydiving/NS_Contrail_80sRibbon/Textures/T_Grid}.RGBA) {maskChannel=Blue, parameterName=Grid Texture Channel, defaultValue=<rgba 0 0 1 0>}
  n9 = Multiply(A=ParticleColor().RGB, B=Add(A=LinearInterpolate(A=@n1.RGB, B=@n2.RGB, Alpha=@n6.R), B=Multiply(A=@n7.RGB, B=@n8.R)))
  MaterialFunctionCall(Emissive=MaterialFunctionCall(LDR Input=Clamp(None=@n9), HDR Tint=@n9) {materialFunction=&DeriveHDRfromLDR}.RemappedRange) {materialFunction=&MF_FX_EmissiveAdjustments}
  n10 = Constant() {r=1}
  n11 = DepthFade(Opacity=@n10, FadeDistance=ScalarParameter() {defaultValue=100, parameterName=DepthFade-Standard_Distance}) {fadeDistanceDefault=20}
  n12 = TextureCoordinate()
  n13 = Saturate(None=Multiply(A=ComponentMask(None=@n12) {r=true}) {constB=8})
  n14 = ScalarParameter() {defaultValue=100, parameterName="Fade Length (S)"}
  n15 = ScalarParameter() {defaultValue=75, parameterName="Fade Offset (S)"}
  Multiply(
   A=Multiply(A=ParticleColor().A, B=ShadingPathSwitch(Default=@n11, Deferred=@n11, Mobile=@n10) {desc=Migrated from FeatureLevelSwitch})
   B=Multiply(
    A=Multiply(A=@n13, B=@n13)
    B=MaterialFunctionCall(Fade Length=@n14, Fade Offset=@n15) {materialFunction=&CameraDepthFade}.Result
   )
  )
  n16 = ScalarParameter() {defaultValue=1.25, parameterName=Overall Scale U}
  n17 = ScalarParameter() {defaultValue=0.5, parameterName=Overall Scale V}
  n18 = ScalarParameter() {defaultValue=-1, parameterName=Overall Pan U}
  n19 = ScalarParameter() {parameterName=Overall Pan V}
  n20 = Add(
   A=Multiply(A=TextureCoordinate() {coordinateIndex=1}, B=AppendVector(A=@n16, B=@n17))
   B=Multiply(A=Time() {bOverride_Period=true, period=10000}, B=AppendVector(A=@n18, B=@n19))
  )
  NamedRerouteDeclaration(None=StaticSwitchParameter(True=MaterialFunctionCall(XY=@n20) {materialFunction=&Swizzle}.YX, False=@n20) {parameterName=Swizzle Textures?}) {name=Base UVs, nodeColor=<rgba 0.0078 0 1 1>}
 ]
}
