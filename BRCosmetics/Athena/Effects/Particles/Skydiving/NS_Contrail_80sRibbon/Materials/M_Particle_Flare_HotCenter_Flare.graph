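# Translucent unlit Niagara sprite material for a hot-centered flare, filed under the 80s ribbon skydiving contrail effect.
# A soft sphere mask, raised to separate powers, forms a bright, partly desaturated core and a dimmer glow, which joins a flare texture tinted by particle color.
# Their sum passes through an emissive adjustment function.
# A second product combines glow plus texture green with particle alpha, depth fade (bypassed on mobile) and switchable near-camera fade.
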Material /BRCosmetics/Athena/Effects/Particles/Skydiving/NS_Contrail_80sRibbon/Materials/M_Particle_Flare_HotCenter_Flare
graph {
 properties {
  bAllowNegativeEmissiveColor=true
  bAutomaticallySetUsageInEditor=false
  bUsedWithNiagaraSprites=true
  blendMode=BLEND_Translucent
  pixelDepthOffsetMode=PDOM_Legacy
  shadingModel=MSM_Unlit
  usedShadingModels=Unlit
 }
 nodes [
  n1 = Constant() {r=0.5}
  n2 = SphereMask(A=TextureCoordinate(), B=@n1, Radius=@n1) {hardnessPercent=0}
  n3 = Multiply(
   A=ScalarParameter() {defaultValue=25, parameterName=CenterBrightness}
   B=Power(Base=@n2, Exp=ScalarParameter() {defaultValue=25, parameterName=CentertPower}) {constExponent=25}
  ) {constB=2}
  n4 = ParticleColor()
  n5 = Desaturation(None=@n4.RGB, Fraction=ScalarParameter() {defaultValue=0.3, parameterName=CenterDesaturation}) {luminanceFactors=<rgba 0.59 0.058 0.4986 0>}
  n6 = Multiply(
   A=Power(Base=@n2, Exp=ScalarParameter() {defaultValue=2, parameterName=GlowPower}) {constExponent=4}
   B=ScalarParameter() {defaultValue=0.15, parameterName=GlowBrightness}
  ) {constB=0.15}
  n7 = TextureSampleParameter2D(UVs=TextureCoordinate()) {parameterName=FlareMask, texture=/Game/Missions/Primary/DeliverTheBomb/Textures/T_DtB_LightningZap}
  n8 = Add(A=@n6, B=@n7.RGB)
  MaterialFunctionCall(Emissive=Add(A=Multiply(A=@n3, B=@n5), B=Multiply(A=@n8, B=@n4.RGB))) {materialFunction=&MF_FX_EmissiveAdjustments}
  n9 = DepthFade(Opacity=@n4.A, FadeDistance=ScalarParameter() {defaultValue=30, parameterName=DepthFade})
  n10 = ShadingPathSwitch(Default=@n9, Deferred=@n9, Mobile=@n4.A) {desc=Migrated from FeatureLevelSwitch}
  n11 = MaterialFunctionCall(Distance=ScalarParameter() {parameterName=CameraFadeDistance}) {materialFunction=&Near_Camera_Fade}
  Multiply(
   A=Add(A=@n6, B=@n7.G)
   B=StaticSwitchParameter(True=Multiply(A=@n10, B=@n11.Result), False=@n10) {group=DetailedGradientMask, parameterName=UseCameraNearFade}
  )
 ]
}
