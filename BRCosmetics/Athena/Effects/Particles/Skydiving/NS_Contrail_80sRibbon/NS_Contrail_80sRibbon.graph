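# 80s-themed Niagara skydiving contrail that trails four ribbons from sockets on the player's skeletal mesh.
# Ribbons use fixed burst particles repositioned every update along an upward offset with sine/cosine waves.
# Flare, Glow, lights, SubUV Flair and GPU Sparkles dress the mesh, while a Landscape mesh pans its material.
# FN_DivingAttributes supplies Dive, which shortens ribbons and speeds panning; time of day scales colour on most emitters.
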NiagaraSystem /BRCosmetics/Athena/Effects/Particles/Skydiving/NS_Contrail_80sRibbon/NS_Contrail_80sRibbon
dependencies {
 usedDataInterfaces [&NiagaraDataInterfaceCurve]
 usedDynamicInputs []
 usedModules [&SystemState, /Game/Effects/Niagara/Modules/Skydiving/FN_DivingAttributes/v6/FN_DivingAttributes]
}
emitters {
 Flair {
  emitterUpdateScript [
   &EmitterState()
   &SpawnRate(SpawnRate=20, Use Spawn Probability=false)
  ]
  particleSpawnScript [
   &InitializeParticle(
    Lifetime Max=0.9
    Lifetime Min=0.4
    Lifetime Mode=::NewEnumerator1
    Sprite Rotation Mode=::NewEnumerator2
    Sprite Size Mode=::NewEnumerator1
    Uniform Sprite Size Max=13
    Uniform Sprite Size Min=10
    UsePositionOffset=false
    Write Material Random=true
   )
   &FN_SetSpawnGroup()
   &SubUVAnimation(End Frame=15, SubUV Animation Mode:ENiagara_SubUVLookupMode=::NewEnumerator2)
   &AddVelocity(
    Cone Angle=30
    Cone Axis=<0 0 1>
    Module.UseConeFalloff=false
    Module.UseVelDistribution=false
    Override Seed=false
    Velocity Mode=::NewEnumerator2
    Velocity Speed=200
   )
   &SkeletalMeshLocation(
    Bone / Socket Index=&SpawnGroup()
    MaskSpawnGroup=false
    Mesh Sampling Type=::NewEnumerator5
    Skeletal Mesh=$User.SkeletalMesh
    Socket Sampling Mode=::NewEnumerator5
   )
   /SetVariables(Particles.SpriteRotation:NiagaraFloat=&MakeCustomFloatFromBool(
     Bool=&SetBoolByFloatComparison(A=$Particles.SubImageIndex, B=12, Comparison Type=::NewEnumerator4)
     True Float=0
     False Float=$Particles.SpriteRotation
    )) {isSetParameters=true}
  ]
  particleUpdateScript [
   &ParticleState()
   &Drag(Drag=1.5, Use Linear Drag=true, Use Rotational Drag=false)
   &SpriteRotationRate(Rotation Rate=&MakeCustomFloatFromBool(
     Bool=&SetBoolByFloatComparison(A=$Particles.SubImageIndex, B=12, Comparison Type=::NewEnumerator4)
     True Float=0
     False Float=&RandomRangeFloat(Minimum=-3, Maximum=3)
    )) {module=SpriteRotationRate001}
   &GravityForce(Gravity=&RandomRangeVector(Minimum=<0 0 200>, Maximum=<1 1 300>))
   &ScaleSpriteSize(Scale Factor=&Vector2DFromFloat(Value=&FloatFromCurve(FloatCurve=DataInterface() {propertyValues {Curve {keys [
          {tangentMode=RCTM_User, time=0, value=0.8}
          {arriveTangent=-0.0405, interpMode=RCIM_Cubic, leaveTangent=-0.0405, tangentMode=RCTM_User, time=0.72, value=1}
          {arriveTangent=-8.0089, interpMode=RCIM_Cubic, leaveTangent=-8.0089, tangentMode=RCTM_User, time=1, value=0}
         ]}}})))
   &SolveForcesAndVelocity(Clamp Velocity=false, Limit Acceleration=false)
  ]
  rendererClasses [&NiagaraSpriteRendererProperties]
  renderers [{rendererClass=&NiagaraSpriteRendererProperties}]
 }
 Flare {
  emitterUpdateScript [
   &EmitterState(Inactive Response=::NewEnumerator1, Life Cycle Mode=::NewEnumerator1, Loop Duration=0.1, UseLoopDelay=false)
   &SpawnBurst_Instantaneous(Spawn Count=4, Use Spawn Probability=false, UseLoopCountLimit=false)
  ]
  particleSpawnScript [
   &FN_SetSpawnGroup()
   &InitializeParticle(
    Color=<rgba 0.4583 0 0.0764 1>
    Lifetime=0.4
    Sprite Rotation Mode=::NewEnumerator2
    Sprite Size=&Multiply_Vector2D_ByFloat(
     Float=&Multiply_Float(
      A=&RandomRangeFloat(Minimum=30, Maximum=40)
      B=&MakeCustomFloatFromBool(
       Bool=&SetBoolByIntComparison(A=&SpawnGroup(), B=2, Comparison Type=::NewEnumerator2)
       True Float=0.75
       False Float=1
      )
     )
     Vector2D=<1 1.5>
    )
    Sprite Size Mode=::NewEnumerator4
    UsePositionOffset=false
    Write Material Random=true
   )
   &CameraOffset(Camera Offset Amount=15)
  ]
  particleUpdateScript [
   &ParticleState()
   &SkeletalMeshLocation(
    Bone / Socket Index=&SpawnGroup()
    MaskSpawnGroup=false
    Mesh Sampling Type=::NewEnumerator5
    Sampled Position Offset=<10 0 0>
    Skeletal Mesh=$User.SkeletalMesh
    Socket Sampling Mode=::NewEnumerator5
   )
   &ScaleSpriteSize(Uniform Curve Sprite Scale=DataInterface() {propertyValues {Curve {keys [
        {interpMode=RCIM_Cubic, tangentMode=RCTM_User, time=0, value=0}
        {interpMode=RCIM_Cubic, tangentMode=RCTM_User, time=0.5, value=1}
        {tangentMode=RCTM_User, time=1, value=0}
       ]}}})
   &FN_Light_Attributes(
    Color=<rgba 0.4583 0 0.0764 1>
    Create Color Attribute=true
    Create Exponent Attribute=true
    Create Radius Attribute=true
    Exponent=4
    Radius=50
    Scale Color=50
   )
   &ScaleColor(
    Scale RGB=&VectorFromFloat(Value=&FNTimeOfDayScalar(Night Scale=0.5, Morning Scale=0.6, Evening Scale=0.6))
    ScaleA=true
    ScaleRGB=true
   )
  ]
  rendererClasses [&NiagaraSpriteRendererProperties, &NiagaraLightRendererProperties]
  renderers [
   {rendererClass=&NiagaraSpriteRendererProperties}
   {rendererClass=&NiagaraLightRendererProperties, rendererIndex=1}
  ]
 }
 Glow {
  emitterUpdateScript [
   &EmitterState(Inactive Response=::NewEnumerator1, Life Cycle Mode=::NewEnumerator1, Loop Duration=0.1, UseLoopDelay=false)
   &SpawnBurst_Instantaneous(Spawn Count=4, Use Spawn Probability=false, UseLoopCountLimit=false)
  ]
  particleSpawnScript [
   &InitializeParticle(
    Color=<rgba 0.4062 0 0.0406 0.5>
    Lifetime=0.25
    Sprite Rotation Mode=::NewEnumerator2
    Sprite Size Mode=::NewEnumerator3
    Uniform Sprite Size=50
    UsePositionOffset=false
    Write Material Random=true
   )
   &FN_SetSpawnGroup()
   &CameraOffset(Camera Offset Amount=15)
  ]
  particleUpdateScript [
   &ParticleState()
   &SkeletalMeshLocation(
    Bone / Socket Index=&SpawnGroup()
    MaskSpawnGroup=false
    Mesh Sampling Type=::NewEnumerator5
    Sampled Position Offset=<10 0 0>
    Skeletal Mesh=$User.SkeletalMesh
    Socket Sampling Mode=::NewEnumerator5
   )
   &ScaleColor(
    Scale RGB=&VectorFromFloat(Value=&FNTimeOfDayScalar(Night Scale=0.5, Morning Scale=0.6, Evening Scale=0.6))
    ScaleA=true
    ScaleRGB=true
   )
  ]
  rendererClasses [&NiagaraSpriteRendererProperties]
  renderers [{rendererClass=&NiagaraSpriteRendererProperties}]
 }
 Landscape {
  emitterUpdateScript [
   &EmitterState(
    Inactive Response=::NewEnumerator1
    Life Cycle Mode=::NewEnumerator1
    Loop Behavior=::NewEnumerator1
    Loop Duration Mode=::NewEnumerator1
    UseLoopDelay=false
   )
   &SpawnBurst_Instantaneous(Use Spawn Probability=false, UseLoopCountLimit=false)
  ]
  particleSpawnScript [
   &InitializeParticle(
    Color=<rgba 1 0 0.1 1>
    Lifetime=5
    Mesh Scale=<200 300 1>
    Mesh Scale Mode=::NewEnumerator4
    Position Offset=<0 0 -30>
    UsePositionOffset=true
    Write Material Random=true
   )
   &InitialMeshOrientation(Mesh Orientation Mode=::NewEnumerator1, Rotation=<0 0.05 0>, Use Rotation Vector=true)
   /SetVariables() {isSetParameters=true}
  ]
  particleUpdateScript [
   &ParticleState(Kill Particles When Lifetime Has Elapsed=false, Let Infinitely Lived Particles Die When Emitter Deactivates=true)
   &ScaleColor(
    Scale RGB=&VectorFromFloat(Value=&FNTimeOfDayScalar(Night Scale=0.5, Morning Scale=0.6, Evening Scale=0.6))
    ScaleA=true
    ScaleRGB=true
   )
   /SetVariables(Particles.PanTime:NiagaraFloat=&Add_Float(
     A=$Particles.PanTime
     B=&Lerp_Float(A=$Engine.DeltaTime, B=&Multiply_Float(A=$Engine.DeltaTime, B=4), Alpha=$System.Dive)
    )) {isSetParameters=true}
   &DynamicMaterialParameters(
    Index 0 Param 1=$Particles.PanTime
    Module.Param0WriteEnabled=true
    Module.Param1WriteEnabled=true
    Module.Param2WriteEnabled=true
    Module.Param3WriteEnabled=true
   )
  ]
  rendererClasses [&NiagaraMeshRendererProperties]
  renderers [{rendererClass=&NiagaraMeshRendererProperties, rendererIndex=0}]
 }
 Ribbon {
  emitterSpawnScript [/SetVariables(Emitter.ParticleCount=20) {isSetParameters=true}]
  emitterUpdateScript [
   &EmitterState(
    Inactive Response=::NewEnumerator1
    Life Cycle Mode=::NewEnumerator1
    Loop Behavior=::NewEnumerator1
    Loop Duration Mode=::NewEnumerator1
    UseLoopDelay=false
   )
   &SpawnBurst_Instantaneous(Spawn Count=$Emitter.ParticleCount, Use Spawn Probability=false, UseLoopCountLimit=false)
   &SpawnBurst_Instantaneous(Spawn Count=$Emitter.ParticleCount, Spawn Group=1, Use Spawn Probability=false, UseLoopCountLimit=false) {module=SpawnBurst_Instantaneous001}
   &SpawnBurst_Instantaneous(Spawn Count=$Emitter.ParticleCount, Spawn Group=2, Use Spawn Probability=false, UseLoopCountLimit=false) {module=SpawnBurst_Instantaneous002}
   &SpawnBurst_Instantaneous(Spawn Count=$Emitter.ParticleCount, Spawn Group=3, Use Spawn Probability=false, UseLoopCountLimit=false) {module=SpawnBurst_Instantaneous003}
  ]
  particleSpawnScript [
   /SetVariables(Particles.RibbonLinkOrder:NiagaraFloat=&ReturnNormalizedExecIndex()) {isSetParameters=true}
   &InitializeRibbon(
    Color=&MakeLinearColorFromVectorAndFloat(
     "Vector (RGB)"=<1 1 1>
     "Float (Alpha)"=&FloatFromCurve(
      FloatCurve=DataInterface() {propertyValues {Curve {keys [
          {tangentMode=RCTM_User, time=0, value=1}
          {arriveTangent=-0.0405, interpMode=RCIM_Cubic, leaveTangent=-0.0405, tangentMode=RCTM_User, time=0.2, value=1}
          {arriveTangent=-5.695, interpMode=RCIM_Cubic, leaveTangent=-5.695, tangentMode=RCTM_User, time=1, value=0}
         ]}}}
      CurveIndex=$Particles.RibbonLinkOrder
      Scale Curve=0.8
     )
    )
    Ribbon Width=&FloatFromCurve(FloatCurve=DataInterface() {propertyValues {Curve {keys [<at 0 0.2>,<at 1 1>]}}}, CurveIndex=$Particles.RibbonLinkOrder, Scale Curve=25)
    UsePositionOffset=false
    Write Material Random=true
   )
  ]
  particleUpdateScript [
   &ParticleState(Kill Particles When Lifetime Has Elapsed=false, Let Infinitely Lived Particles Die When Emitter Deactivates=true)
   &SkeletalMeshLocation(
    Bone / Socket Index=&SpawnGroup()
    "Interpolate Mesh Sampling (Update)"=true
    MaskSpawnGroup=false
    Mesh Sampling Type=::NewEnumerator5
    Position Sampling=::NewEnumerator2
    Sampled Position Offset=<10 0 0>
    Skeletal Mesh=$User.SkeletalMesh
    Socket Sampling Mode=::NewEnumerator5
   )
   /SetVariables(Particles.Position=/Niagara/DynamicInputs/Lerp/Lerp_Position(
     A:NiagaraPosition=$Particles.SkeletalMeshLocation.SampledPosition
     B:NiagaraPosition=&Add_Vector(
      A=Linked() {linkedVariable {name=Particles.SkeletalMeshLocation.SampledPosition, type=NiagaraPosition}}
      B=&Lerp_Vector(A=<0 0 200>, B=<0 0 100>, Alpha=$System.Dive)
     )
     Alpha:NiagaraFloat=$Particles.RibbonLinkOrder
    )) {isSetParameters=true}
   /SetVariables(Particles.Position=&Add_Vector(
     A=Linked() {linkedVariable {name=Particles.Position, type=NiagaraPosition}}
     B=&MakeVector(
      X=&Sine(
       Normalized Angle=&Add_Float(
        A=&Add_Float(A=$Particles.RibbonLinkOrder, B=&Multiply_Float(A=$Emitter.Age, B=-1))
        B=&Multiply_FloatByInt(Float=0.33, Integer=&SpawnGroup())
       )
       Period=2
       Scale=&Multiply_Float(A=$Particles.RibbonLinkOrder, B=20)
       Bias=0
      )
      Y=&Cosine(
       Normalized Angle=&Add_Float(
        A=&Add_Float(A=$Particles.RibbonLinkOrder, B=&Multiply_Float(A=$Emitter.Age, B=-1))
        B=&Multiply_FloatByInt(Float=0.33, Integer=&SpawnGroup())
       )
       Period=1.3
       Scale=&Multiply_Float(A=$Particles.RibbonLinkOrder, B=18)
       Bias=0
      )
      Z=0
     )
    )) {isSetParameters=true}
   &ScaleColor(
    Scale RGB=&VectorFromFloat(Value=&FNTimeOfDayScalar(Night Scale=0.5, Morning Scale=0.6, Evening Scale=0.6))
    ScaleA=true
    ScaleRGB=true
   )
  ]
  rendererClasses [&NiagaraRibbonRendererProperties]
  renderers [{rendererClass=&NiagaraRibbonRendererProperties}]
 }
 Sparkles {
  simTarget=GPUComputeSim
  emitterUpdateScript [
   &EmitterState()
   &SpawnRate(SpawnRate=30, Use Spawn Probability=false)
  ]
  particleSpawnScript [
   &InitializeParticle(
    Lifetime Max=0.9
    Lifetime Min=0.4
    Lifetime Mode=::NewEnumerator1
    Sprite Rotation Mode=::NewEnumerator2
    Sprite Size Mode=::NewEnumerator1
    UsePositionOffset=false
    Write Material Random=true
   )
   &FN_SetSpawnGroup() {module=FN_SetSpawnGroup001}
   &AddVelocity(
    Cone Angle=30
    Cone Axis=<0 0 1>
    Module.UseConeFalloff=false
    Module.UseVelDistribution=false
    Override Seed=false
    Velocity Mode=::NewEnumerator2
    Velocity Speed=300
   )
   &SkeletalMeshLocation(MaskSpawnGroup=false, Mesh Sampling Type=::NewEnumerator5, OverrideRandomSeed=false, Skeletal Mesh=$User.SkeletalMesh) {module=SkeletalMeshLocation001}
  ]
  particleUpdateScript [
   &ParticleState()
   &Drag(Drag=2, Use Linear Drag=true, Use Rotational Drag=false)
   &GravityForce(Gravity=&RandomRangeVector(Minimum=<0 0 200>, Maximum=<1 1 300>))
   &ScaleSpriteSize(Scale Factor=&Vector2DFromFloat(Value=&FloatFromCurve(FloatCurve=DataInterface() {propertyValues {Curve {keys [<at 0.0018 0.1957>,<at 0.0527 0.819>,<at 0.1169 0.6555>,<at 0.1748 0.9015>,<at 0.2394 0.582>,<at 0.332 1.0023>,<at 0.4947 0.3707>,<at 0.6727 1>,<at 1 0>]}}})))
   &SolveForcesAndVelocity(Clamp Velocity=false, Limit Acceleration=false)
  ]
  rendererClasses [&NiagaraSpriteRendererProperties]
  renderers [{rendererClass=&NiagaraSpriteRendererProperties}]
 }
}
summary {
 systemName=NS_Contrail_80sRibbon
 emitters [
  {emitterName=Flair, rendererClasses [&NiagaraSpriteRendererProperties]}
  {emitterName=Flare, rendererClasses [&NiagaraSpriteRendererProperties, &NiagaraLightRendererProperties]}
  {emitterName=Sparkles, simTarget=GPUComputeSim, rendererClasses [&NiagaraSpriteRendererProperties]}
  {emitterName=Ribbon, rendererClasses [&NiagaraRibbonRendererProperties]}
  {emitterName=Glow, rendererClasses [&NiagaraSpriteRendererProperties]}
  {emitterName=Landscape, rendererClasses [&NiagaraMeshRendererProperties]}
 ]
 userVariables [
  User.PlayerSpeed {type=NiagaraFloat, defaultValue=0, description=""}
  User.SkeletalMesh {
   type=NiagaraDataInterfaceSkeletalMesh
   defaultValue {
    struct=NiagaraExt_VariableValue_DataInterface
    value {dataInterface=&self:NiagaraDataInterfaceSkeletalMesh_0}
   }
  }
 ]
}
